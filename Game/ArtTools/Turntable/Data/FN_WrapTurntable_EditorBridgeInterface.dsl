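; Blueprint function set FN_WrapTurntable_EditorBridgeInterface, the interface between the wrap turntable art tool and the editor.
; It declares gotoWrapsFolder, which returns a boolean, and loadSelectedWraps, which returns a numeric value.
; Both bodies hold only default returns (false and zero), so the file itself carries no logic.
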(fn gotoWrapsFolder
  (return false))

(fn loadSelectedWraps
  (return 0))
